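# Additive, unlit, two-sided particle material for audio-reactive shapes in the Papaya event, masked by a diamond texture by default.
# Event audio amplitude and spectrum, or a debug sine, drive two summed vertex-normal offsets (speaker push and inverse ripple), scaled by particle red.
# Particle green and blue weight two spectrum bands in the ripple.
# MF_Papaya_GadgetColor supplies color, which is hue-shifted by amplitude, masked, boosted to HDR and multiplied by particle alpha.
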Material /FNGameplayCues/Athena/Effects/Materials/Events/Papaya/M_Papaya_DF_Shapes
graph {
 properties {
  bUsedWithParticleSprites=true
  blendMode=BLEND_Additive
  pixelDepthOffsetMode=PDOM_Legacy
  shadingModel=MSM_Unlit
  twoSided=true
  usedShadingModels=Unlit
 }
 nodes [
  n1 = ParticleColor()
  n2 = ScalarParameter() {defaultValue=1, parameterName=Vert Color Power}
  n3 = VertexNormalWS()
  n4 = CollectionParameter() {collection=&FortniteMaterialParameters, parameterName=EventAudio_Amplitude}
  n5 = Multiply(
   A=MaterialFunctionCall() {materialFunction=&DebugTimeSine}.Result
   B=ScalarParameter() {defaultValue=10, parameterName=Debug Sine Multiplier}
  )
  n6 = ScalarParameter() {parameterName=Debug}
  n7 = LinearInterpolate(A=@n4, B=@n5, Alpha=@n6)
  n8 = CollectionParameter() {collection=&FortniteMaterialParameters, parameterName=EventAudio_Spectrum}
  n9 = LinearInterpolate(A=@n8, B=@n5, Alpha=@n6)
  n10 = ComponentMask(None=@n9) {r=true}
  n11 = Add(A=@n10, B=@n7)
  n12 = ScalarParameter() {defaultValue=3, parameterName=Average Amplitude Multiplier 2}
  n13 = Multiply(
   A=Power(Base=Multiply(A=@n11, B=@n12), Exp=ScalarParameter() {defaultValue=1, parameterName=Average Amplitude Power 2})
   B=ScalarParameter() {defaultValue=12, parameterName=SpeakerOffset}
  )
  n14 = LinearInterpolate(Alpha=Multiply(A=Power(Base=@n7) {constExponent=4}, B=ScalarParameter() {defaultValue=1, parameterName=Amplitude Multiplier})) {constB=0.2}
  n15 = ParticleColor()
  n16 = Add(A=Multiply(A=@n10, B=@n15.G), B=Multiply(A=@n15.B, B=ComponentMask(None=@n9) {g=true}))
  n17 = ScalarParameter() {defaultValue=1, parameterName=Secoondary Multiplier}
  n18 = ComponentMask(None=Add(A=@n14, B=Multiply(A=@n16, B=@n17))) {r=true}
  n19 = Multiply(A=Power(Base=OneMinus(None=@n18), Exp=ScalarParameter() {defaultValue=1, parameterName=Ripple Alpha Power}), B=@n1.R)
  n20 = ScalarParameter() {defaultValue=40, parameterName=Sine Wave Height}
  Add(
   A=Multiply(A=Power(Base=@n1.R, Exp=@n2) {constExponent=1.25}, B=Multiply(A=@n3, B=@n13))
   B=Multiply(A=VertexNormalWS(), B=Multiply(A=@n19, B=@n20))
  )
  n21 = MaterialFunctionCall!(
   Media Player TexCoord
   Media Color Code Coordinate
   Media Texture Resolution
   Base Color Parameter
   Base Color Mix
   Media Color Code Mix
   Intermission Color Parameter
   Intermission Color Mix
   Desaturation
  ) {materialFunction=/Game/Environments/Papaya/Materials/MaterialFunctions/MF_Papaya_GadgetColor}
  n22 = Power(
   Base=TextureSampleParameter2D() {parameterName=Texture Mask, texture=/Game/Effects/Fort_Effects/PlayerHealthDamage/T_Diamond}.RGB
   Exp=ScalarParameter() {defaultValue=1, parameterName=Mask Power}
  )
  n23 = Saturate(None=Multiply(
    A=MaterialFunctionCall(Hue Shift Percentage=@n14, Texture=@n21.Result) {materialFunction=&HueShift}.Result
    B=@n22
   ))
  n24 = ScalarParameter() {defaultValue=4, parameterName=HDR Intensity}
  MaterialFunctionCall(Emissive=Multiply(
    A=MaterialFunctionCall(LDR Input=@n23, HDR Tint=@n22, HDR Intensity Multiplier=@n24) {materialFunction=&DeriveHDRfromLDR}.RemappedRange
    B=ParticleColor().A
   )) {materialFunction=&MF_FX_EmissiveAdjustments}
 ]
}
